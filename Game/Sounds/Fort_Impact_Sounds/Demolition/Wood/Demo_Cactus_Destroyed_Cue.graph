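# Sound cue for a cactus being destroyed, with medium-object demolition attenuation and a four-max impact concurrency set that prevents new sounds.
# A quality switch sends the first two levels to a mixer and the last two to a pitch-varied, optimized tree destruction wave.
# The mixer plays one of two random cactus impacts without pitch variation, as its second input is a modulator with no child.
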SoundCue /Game/Sounds/Fort_Impact_Sounds/Demolition/Wood/Demo_Cactus_Destroyed_Cue
root=QualityLevel_0
nodes [
 {id=Mixer_6, children [Random_0,Modulator_8], properties {inputVolume [1,1]}}
 {id=Modulator_0, children [WavePlayer_0], properties {pitchMax=1.1, pitchMin=0.9}}
 {id=Modulator_8, children [null], properties {pitchMax=1.1, pitchMin=0.9}}
 {id=QualityLevel_0, children [Mixer_6,Mixer_6,Modulator_0,Modulator_0]}
 {id=Random_0, children [WavePlayer_1,WavePlayer_2], properties {weights [1,1]}}
 {id=WavePlayer_0, properties {soundWaveAssetPtr=/Game/Sounds/Fort_Impact_Sounds/Demolition/Tree/Tree_Destroy_opt}}
 {id=WavePlayer_1, properties {soundWaveAssetPtr=&/Cactus_Impact_01}}
 {id=WavePlayer_2, properties {soundWaveAssetPtr=&/Cactus_Impact_02}}
]
properties {
 attenuationSettings=/Game/Sounds/Attenuation/Fort_Demolition_Destroy_Medium_Sized_Object_Att
 maxDistance=4000
 soundClassObject=&Fort_Impacts
 volumeMultiplier=0.68
 concurrencyOverrides {resolutionRule=PreventNew}
 concurrencySet [&CONC_Object_Impacts_FourMax]
}
