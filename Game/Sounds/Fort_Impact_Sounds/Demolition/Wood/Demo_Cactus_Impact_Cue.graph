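# Sound cue for cactus demolition impacts, playing randomized hit sounds with small-object demolition attenuation in the impacts sound class.
# A quality-level root sends its first slot to an equal-weight random pick of four waves, while the other three slots share a lighter two-wave branch.
# Each branch passes through a modulator for slight pitch variation.
# An object-impact concurrency set limits simultaneous playback.
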SoundCue /Game/Sounds/Fort_Impact_Sounds/Demolition/Wood/Demo_Cactus_Impact_Cue
root=QualityLevel_0
nodes [
 {id=Modulator_0, children [Random_3], properties {pitchMax=1.11, pitchMin=0.9}}
 {id=Modulator_2, children [Random_2], properties {pitchMax=1.07, pitchMin=0.92}}
 {id=QualityLevel_0, children [Modulator_2,Modulator_0,Modulator_0,Modulator_0]}
 {id=Random_2, children [WavePlayer_5,WavePlayer_6,WavePlayer_7,WavePlayer_8], properties {weights [1,1,1,1]}}
 {id=Random_3, children [WavePlayer_9, WavePlayer_10], properties {weights [1,1]}}
 {id=WavePlayer_10, properties {soundWaveAssetPtr=&/Cactus_Impact_02}}
 {id=WavePlayer_5, properties {soundWaveAssetPtr=&/Cactus_Impact_01}}
 {id=WavePlayer_6, properties {soundWaveAssetPtr=&Cactus_Impact_02}}
 {id=WavePlayer_7, properties {soundWaveAssetPtr=&/Cactus_Impact_03}}
 {id=WavePlayer_8, properties {soundWaveAssetPtr=&/Cactus_Impact_04}}
 {id=WavePlayer_9, properties {soundWaveAssetPtr=&Cactus_Impact_01}}
]
properties {
 attenuationSettings=/Game/Sounds/Attenuation/Fort_Demolition_Hit_Small_Sized_Object_Att
 maxDistance=4000
 soundClassObject=&Fort_Impacts
 volumeMultiplier=0.68
 concurrencyOverrides {resolutionRule=PreventNew}
 concurrencySet [&CONC_Object_Impacts_FourMax]
}
