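# Sound cue for small wood impacts, named for demolition hits on a wooden chest, in the impacts sound class.
# A quality level root picks one of four branches, stepping from four random wave variations down to three, two and a single wave.
# Each branch runs through a modulator, with the single-wave branch setting its own pitch range to add variation.
# Small-object demolition attenuation and a shared object-impact concurrency set govern playback.
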SoundCue /Game/Sounds/Fort_Impact_Sounds/Demolition/Wood/Demo_Chest_Wood_Small_Impact_Cue
root=QualityLevel_0
nodes [
 {id=Modulator_1, children [Random_0]}
 {id=Modulator_3, children [Random_1]}
 {id=Modulator_4, children [Random_2]}
 {id=Modulator_5, children [WavePlayer_10], properties {pitchMax=1.15, pitchMin=0.9}}
 {id=QualityLevel_0, children [Modulator_1,Modulator_3,Modulator_4,Modulator_5]}
 {id=Random_0, children [WavePlayer_0,WavePlayer_1,WavePlayer_2,WavePlayer_3], properties {weights [1,1,1,1]}}
 {id=Random_1, children [WavePlayer_4,WavePlayer_5,WavePlayer_6], properties {weights [1,1,1]}}
 {id=Random_2, children [WavePlayer_8,WavePlayer_9], properties {weights [1,1]}}
 {id=WavePlayer_0, properties {soundWaveAssetPtr=&/Wood_Small_Impact_mix_01}}
 {id=WavePlayer_1, properties {soundWaveAssetPtr=&/Wood_Small_Impact_mix_02}}
 {id=WavePlayer_10, properties {soundWaveAssetPtr=&Wood_Small_Impact_mix_01}}
 {id=WavePlayer_2, properties {soundWaveAssetPtr=&/Wood_Small_Impact_mix_03}}
 {id=WavePlayer_3, properties {soundWaveAssetPtr=&/Wood_Small_Impact_mix_04}}
 {id=WavePlayer_4, properties {soundWaveAssetPtr=&Wood_Small_Impact_mix_01}}
 {id=WavePlayer_5, properties {soundWaveAssetPtr=&Wood_Small_Impact_mix_02}}
 {id=WavePlayer_6, properties {soundWaveAssetPtr=&Wood_Small_Impact_mix_03}}
 {id=WavePlayer_8, properties {soundWaveAssetPtr=&Wood_Small_Impact_mix_01}}
 {id=WavePlayer_9, properties {soundWaveAssetPtr=&Wood_Small_Impact_mix_02}}
]
properties {
 attenuationSettings=/Game/Sounds/Attenuation/Fort_Demolition_Hit_Small_Sized_Object_Att
 maxDistance=4000
 soundClassObject=&Fort_Impacts
 volumeMultiplier=0.52
 concurrencyOverrides {maxCount=3, resolutionRule=StopOldest}
 concurrencySet [&CONC_Object_Impacts_FourMax]
}
