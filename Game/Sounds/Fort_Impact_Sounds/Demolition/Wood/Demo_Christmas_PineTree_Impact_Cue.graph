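# Sound cue for impacts on a Christmas pine tree during demolition, played as an attenuated impact sound under a shared object-impact concurrency set.
# A quality-level root picks among four branches of decreasing variety.
# The first mixes random holiday-tree and pine-hit waves; the second fixes the holiday layer to one wave.
# The last two each play one optimized combined wave, and every layer gets randomized pitch, with some also varying volume.
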SoundCue /Game/Sounds/Fort_Impact_Sounds/Demolition/Wood/Demo_Christmas_PineTree_Impact_Cue
root=QualityLevel_0
nodes [
 {id=Mixer_0, children [Modulator_0,Modulator_2], properties {inputVolume [1,1]}}
 {id=Mixer_1, children [Modulator_3,Modulator_1], properties {inputVolume [1,1]}}
 {id=Modulator_0, children [Random_0], properties {pitchMax=1.1, pitchMin=0.9, volumeMin=0.65}}
 {id=Modulator_1, children [Random_3], properties {pitchMax=1.07, pitchMin=0.92}}
 {id=Modulator_2, children [Random_1], properties {pitchMax=1.07, pitchMin=0.92}}
 {id=Modulator_3, children [WavePlayer_6], properties {pitchMax=1.1, pitchMin=0.9, volumeMin=0.65}}
 {id=Modulator_4, children [WavePlayer_15], properties {pitchMax=1.07, pitchMin=0.88}}
 {id=Modulator_5, children [WavePlayer_16], properties {pitchMax=1.07, pitchMin=0.88}}
 {id=QualityLevel_0, children [Mixer_0,Mixer_1,Modulator_4,Modulator_5]}
 {id=Random_0, children [WavePlayer_0,WavePlayer_1,WavePlayer_2], properties {weights [1,1,1]}}
 {id=Random_1, children [WavePlayer_3,WavePlayer_4,WavePlayer_5], properties {weights [1,1,1]}}
 {id=Random_3, children [WavePlayer_9, WavePlayer_10, WavePlayer_11], properties {weights [1,1,1]}}
 {id=WavePlayer_0, properties {soundWaveAssetPtr=/Game/Sounds/Fort_Impact_Sounds/Demolition/Tree/Trees_Holiday_Winter_2017-001}}
 {id=WavePlayer_1, properties {soundWaveAssetPtr=/Game/Sounds/Fort_Impact_Sounds/Demolition/Tree/Trees_Holiday_Winter_2017-002}}
 {id=WavePlayer_10, properties {soundWaveAssetPtr=/Game/Sounds/Fort_Impact_Sounds/Demolition/Tree/tree_pine_hit_02}}
 {id=WavePlayer_11, properties {soundWaveAssetPtr=/Game/Sounds/Fort_Impact_Sounds/Demolition/Tree/tree_pine_hit_03}}
 {id=WavePlayer_15, properties {soundWaveAssetPtr=/Game/Sounds/Fort_Impact_Sounds/Demolition/Tree/PineTree_Impact_Christmas_opt}}
 {id=WavePlayer_16, properties {soundWaveAssetPtr=&PineTree_Impact_Christmas_opt}}
 {id=WavePlayer_2, properties {soundWaveAssetPtr=/Game/Sounds/Fort_Impact_Sounds/Demolition/Tree/Trees_Holiday_Winter_2017-003}}
 {id=WavePlayer_3, properties {soundWaveAssetPtr=/Game/Sounds/Fort_Impact_Sounds/Demolition/Tree/tree_pine_hit_01}}
 {id=WavePlayer_4, properties {soundWaveAssetPtr=&tree_pine_hit_02}}
 {id=WavePlayer_5, properties {soundWaveAssetPtr=&tree_pine_hit_03}}
 {id=WavePlayer_6, properties {soundWaveAssetPtr=&Trees_Holiday_Winter_2017-001}}
 {id=WavePlayer_9, properties {soundWaveAssetPtr=&tree_pine_hit_01}}
]
properties {
 attenuationSettings=/Game/Sounds/Attenuation/Fort_Demolition_Hit_Small_Sized_Object_Att
 maxDistance=4000
 soundClassObject=&Fort_Impacts
 volumeMultiplier=0.34
 concurrencyOverrides {resolutionRule=PreventNew}
 concurrencySet [&CONC_Object_Impacts_FourMax]
}
