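# Sound cue for mushroom demolition impacts, routed to the impacts sound class with small-object hit attenuation.
# A quality-level node sends its first branch to a random pick of five waves, while the other three share a random pick of two.
# Each branch applies its own slight pitch modulation for variation.
# An object-impact concurrency set is assigned, alongside a prevent-new override rule.
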SoundCue /Game/Sounds/Fort_Impact_Sounds/Demolition/Wood/Demo_Mushroom_Impact_Cue
root=QualityLevel_0
nodes [
 {id=Modulator_0, children [Random_3], properties {pitchMax=1.11, pitchMin=0.9}}
 {id=Modulator_2, children [Random_0], properties {pitchMax=1.07, pitchMin=0.92}}
 {id=QualityLevel_0, children [Modulator_2,Modulator_0,Modulator_0,Modulator_0]}
 {id=Random_0, children [WavePlayer_0,WavePlayer_1,WavePlayer_2,WavePlayer_3,WavePlayer_4], properties {weights [1,1,1,1,1]}}
 {id=Random_3, children [WavePlayer_11, WavePlayer_12], properties {weights [1,1]}}
 {id=WavePlayer_0, properties {soundWaveAssetPtr=&/Demo_Mushroom_Impact_01}}
 {id=WavePlayer_1, properties {soundWaveAssetPtr=&/Demo_Mushroom_Impact_02}}
 {id=WavePlayer_11, properties {soundWaveAssetPtr=&Demo_Mushroom_Impact_01}}
 {id=WavePlayer_12, properties {soundWaveAssetPtr=&Demo_Mushroom_Impact_02}}
 {id=WavePlayer_2, properties {soundWaveAssetPtr=/Game/Sounds/Fort_Impact_Sounds/Demolition/Mushroom/Demo_Mushroom_Impact_03}}
 {id=WavePlayer_3, properties {soundWaveAssetPtr=/Game/Sounds/Fort_Impact_Sounds/Demolition/Mushroom/Demo_Mushroom_Impact_04}}
 {id=WavePlayer_4, properties {soundWaveAssetPtr=/Game/Sounds/Fort_Impact_Sounds/Demolition/Mushroom/Demo_Mushroom_Impact_05}}
]
properties {
 attenuationSettings=/Game/Sounds/Attenuation/Fort_Demolition_Hit_Small_Sized_Object_Att
 maxDistance=4000
 soundClassObject=&Fort_Impacts
 volumeMultiplier=1.2
 concurrencyOverrides {resolutionRule=PreventNew}
 concurrencySet [&CONC_Object_Impacts_FourMax]
}
