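# Sound cue for wooden barrel demolition impacts, layering generic wood impacts with barrel-specific hits.
# A mixer blends two randomized three-wave layers, with the barrel-specific layer at half the input volume of the generic one.
# Each layer has its own pitch modulation, skewed higher for the generic wood layer.
# The cue uses the impacts sound class, small-object demolition attenuation and a shared object-impact concurrency set to limit simultaneous hits.
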SoundCue /Game/Sounds/Fort_Impact_Sounds/Demolition/Wood/Demo_WoodBarrel_Impact_Cue
root=Mixer_4
nodes [
 {id=Mixer_4, children [Modulator_2,Modulator_1], properties {inputVolume [1,0.5]}}
 {id=Modulator_1, children [Random_31], properties {pitchMax=1.07, pitchMin=0.92}}
 {id=Modulator_2, children [Random_32], properties {pitchMax=1.2, pitchMin=1}}
 {id=Random_31, children [WavePlayer_202, WavePlayer_203, WavePlayer_204], properties {weights [1,1,1]}}
 {id=Random_32, children [WavePlayer_206, WavePlayer_207, WavePlayer_209], properties {weights [1,1,1]}}
 {id=WavePlayer_202, properties {soundWaveAssetPtr=&/WoodBarrel_Impact01}}
 {id=WavePlayer_203, properties {soundWaveAssetPtr=&/WoodBarrel_Impact02}}
 {id=WavePlayer_204, properties {soundWaveAssetPtr=&/WoodBarrel_Impact03}}
 {id=WavePlayer_206, properties {soundWaveAssetPtr=/Game/Sounds/Fort_Impact_Sounds/Demolition/Defaults/wood_impact_06}}
 {id=WavePlayer_207, properties {soundWaveAssetPtr=/Game/Sounds/Fort_Impact_Sounds/Demolition/Defaults/wood_impact_07}}
 {id=WavePlayer_209, properties {soundWaveAssetPtr=/Game/Sounds/Fort_Impact_Sounds/Demolition/Defaults/wood_impact_09}}
]
properties {
 attenuationSettings=/Game/Sounds/Attenuation/Fort_Demolition_Hit_Small_Sized_Object_Att
 maxDistance=4000
 soundClassObject=&Fort_Impacts
 volumeMultiplier=0.48
 concurrencyOverrides {resolutionRule=PreventNew}
 concurrencySet [&CONC_Object_Impacts_FourMax]
}
