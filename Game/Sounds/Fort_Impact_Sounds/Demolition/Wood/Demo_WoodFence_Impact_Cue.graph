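# Sound cue for a wood fence demolition impact, in the Fort impacts sound class with small-object hit attenuation.
# A mixer layers one of three random fence impact waves, lightly pitch-varied, with a quieter random pick from three default impact wobble sounds.
# The wobble passes through two stacked pitch modulators, one lowering and one raising, for a net higher, varied pitch.
# A shared object-impact concurrency set is assigned to the cue.
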SoundCue /Game/Sounds/Fort_Impact_Sounds/Demolition/Wood/Demo_WoodFence_Impact_Cue
root=Mixer_5
nodes [
 {id=Mixer_5, children [Modulator_16,Modulator_17], properties {inputVolume [1,0.25]}}
 {id=Modulator_14, children [Random_23], properties {pitchMax=0.8, pitchMin=0.65}}
 {id=Modulator_16, children [Random_0], properties {pitchMax=1.07, pitchMin=0.92}}
 {id=Modulator_17, children [Modulator_14], properties {pitchMax=2, pitchMin=1.9}}
 {id=Random_0, children [WavePlayer_2,WavePlayer_3,WavePlayer_4], properties {weights [1,1,1]}}
 {id=Random_23, children [WavePlayer_75, WavePlayer_76, WavePlayer_77], properties {bRandomizeWithoutReplacement=false, weights [1,1,1]}}
 {id=WavePlayer_2, properties {soundWaveAssetPtr=&/Wood_Fence_Impact03}}
 {id=WavePlayer_3, properties {soundWaveAssetPtr=&/Wood_Fence_Impact04}}
 {id=WavePlayer_4, properties {soundWaveAssetPtr=&/Wood_Fence_Impact05}}
 {id=WavePlayer_75, properties {soundWaveAssetPtr=/Game/Sounds/Fort_Impact_Sounds/Demolition/Defaults/Fort_Impact_Wobble_01}}
 {id=WavePlayer_76, properties {soundWaveAssetPtr=/Game/Sounds/Fort_Impact_Sounds/Demolition/Defaults/Fort_Impact_Wobble_02}}
 {id=WavePlayer_77, properties {soundWaveAssetPtr=/Game/Sounds/Fort_Impact_Sounds/Demolition/Defaults/Fort_Impact_Wobble_03}}
]
properties {
 attenuationSettings=/Game/Sounds/Attenuation/Fort_Demolition_Hit_Small_Sized_Object_Att
 maxDistance=4000
 soundClassObject=&Fort_Impacts
 volumeMultiplier=0.56
 concurrencyOverrides {maxCount=3, resolutionRule=StopOldest}
 concurrencySet [&CONC_Object_Impacts_FourMax]
}
