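# Sound cue for the final demolition impact of a wooden pallet.
# A mixer layers three randomized branches: pitch-varied final-impact pallet waves, lower-pitched shared default explosion impact layers at slightly reduced volume, and regular pallet impact waves without modulation.
# It uses medium-sized object destruction attenuation, the impacts sound class, and a four-voice concurrency limit.
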SoundCue /Game/Sounds/Fort_Impact_Sounds/Demolition/Wood/Demo_WoodPallet_ImpactFinal_Cue
root=Mixer_2
nodes [
 {id=Mixer_2, children [Modulator_2,Modulator_3,Random_5], properties {inputVolume [1,0.8,1]}}
 {id=Modulator_2, children [Random_58], properties {pitchMax=1.07, pitchMin=0.92}}
 {id=Modulator_3, children [Random_57], properties {pitchMax=0.95, pitchMin=0.85}}
 {id=Random_5, children [WavePlayer_17, WavePlayer_21], properties {weights [1,1]}}
 {id=Random_57, children [WavePlayer_181, WavePlayer_182, WavePlayer_183], properties {weights [1,1,1]}}
 {id=Random_58, children [WavePlayer_184, WavePlayer_185, WavePlayer_186], properties {weights [1,1,1]}}
 {id=WavePlayer_17, properties {soundWaveAssetPtr=&/Wood_Pallets_Impact04}}
 {id=WavePlayer_181, properties {soundWaveAssetPtr=/Game/Sounds/Fort_Impact_Sounds/Demolition/Defaults/Explo_Impact_Layer01}}
 {id=WavePlayer_182, properties {soundWaveAssetPtr=/Game/Sounds/Fort_Impact_Sounds/Demolition/Defaults/Explo_Impact_Layer02}}
 {id=WavePlayer_183, properties {soundWaveAssetPtr=/Game/Sounds/Fort_Impact_Sounds/Demolition/Defaults/Explo_Impact_Layer03}}
 {id=WavePlayer_184, properties {soundWaveAssetPtr=&/Wood_Pallets_ImpactFinal03}}
 {id=WavePlayer_185, properties {soundWaveAssetPtr=&/Wood_Pallets_ImpactFinal01}}
 {id=WavePlayer_186, properties {soundWaveAssetPtr=&/Wood_Pallets_ImpactFinal02}}
 {id=WavePlayer_21, properties {soundWaveAssetPtr=&/Wood_Pallets_Impact02}}
]
properties {
 attenuationSettings=/Game/Sounds/Attenuation/Fort_Demolition_Destroy_Medium_Sized_Object_Att
 maxDistance=4000
 soundClassObject=&Fort_Impacts_PSM
 volumeMultiplier=0.48
 concurrencyOverrides {maxCount=4, resolutionRule=StopOldest}
 concurrencySet [&CONC_Object_Impacts_FourMax]
}
