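# Sound cue for the impact of hitting a wood pallet in demolition, played in the Fort_Impacts class with small-object demolition attenuation.
# A quality-level node selects among four pitch-modulated branches, stepping from a random pick of three waves to two, then to one fixed wave for the last two levels.
# The single-wave branches get a slightly wider pitch range.
# A shared object-impact concurrency set is assigned to limit overlapping hits.
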SoundCue /Game/Sounds/Fort_Impact_Sounds/Demolition/Wood/Demo_WoodPallet_Impact_Cue
root=QualityLevel_0
nodes [
 {id=Modulator_0, children [Random_0], properties {pitchMax=1.07, pitchMin=0.92}}
 {id=Modulator_1, children [WavePlayer_3], properties {pitchMax=1.1, pitchMin=0.9}}
 {id=Modulator_16, children [Random_4], properties {pitchMax=1.07, pitchMin=0.92}}
 {id=Modulator_3, children [WavePlayer_10], properties {pitchMax=1.1, pitchMin=0.9}}
 {id=QualityLevel_0, children [Modulator_16,Modulator_0,Modulator_1,Modulator_3]}
 {id=Random_0, children [WavePlayer_0,WavePlayer_1], properties {weights [1,1]}}
 {id=Random_4, children [WavePlayer_15, WavePlayer_16, WavePlayer_7], properties {weights [1,1,1]}}
 {id=WavePlayer_0, properties {soundWaveAssetPtr=&/Wood_Pallets_Impact04}}
 {id=WavePlayer_1, properties {soundWaveAssetPtr=&/Wood_Pallets_Impact02}}
 {id=WavePlayer_10, properties {soundWaveAssetPtr=&Wood_Pallets_Impact04}}
 {id=WavePlayer_15, properties {soundWaveAssetPtr=&Wood_Pallets_Impact04}}
 {id=WavePlayer_16, properties {soundWaveAssetPtr=&Wood_Pallets_Impact02}}
 {id=WavePlayer_3, properties {soundWaveAssetPtr=&Wood_Pallets_Impact04}}
 {id=WavePlayer_7, properties {soundWaveAssetPtr=&/Wood_Pallets_Impact03}}
]
properties {
 attenuationSettings=/Game/Sounds/Attenuation/Fort_Demolition_Hit_Small_Sized_Object_Att
 maxDistance=4000
 soundClassObject=&Fort_Impacts
 volumeMultiplier=0.48
 concurrencyOverrides {maxCount=3, resolutionRule=StopOldest}
 concurrencySet [&CONC_Object_Impacts_FourMax]
}
