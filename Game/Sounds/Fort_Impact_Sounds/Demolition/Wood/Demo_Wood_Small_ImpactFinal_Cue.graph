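# Sound cue for the final impact of demolishing small wood objects, with medium-object destroy attenuation and concurrency-limited playback.
# A quality-level node selects among four branches of decreasing complexity.
# The first two mix pitched-up random shatters, a briefly delayed second shatter and a quieter stereo wood impact, with fewer variations in the second.
# The last two each play a single pre-optimized impact wave with pitch modulation.
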SoundCue /Game/Sounds/Fort_Impact_Sounds/Demolition/Wood/Demo_Wood_Small_ImpactFinal_Cue
root=QualityLevel_0
nodes [
 {id=Delay_0, children [Modulator_1], properties {delayMax=0.12, delayMin=0.1}}
 {id=Delay_6, children [Modulator_25], properties {delayMax=0.12, delayMin=0.1}}
 {id=Mixer_0, children [Modulator_0,Delay_0,Modulator_2], properties {inputVolume [1,1,0.7]}}
 {id=Mixer_7, children [Modulator_24,Delay_6,Random_40], properties {inputVolume [1,1,0.7]}}
 {id=Modulator_0, children [Random_2], properties {pitchMax=1.5, pitchMin=1.2}}
 {id=Modulator_1, children [Random_1], properties {pitchMax=1.4, pitchMin=1.2}}
 {id=Modulator_2, children [WavePlayer_1], properties {pitchMin=0.9}}
 {id=Modulator_24, children [Random_42], properties {pitchMax=1.5, pitchMin=1.2}}
 {id=Modulator_25, children [Random_41], properties {pitchMax=1.4, pitchMin=1.2}}
 {id=Modulator_3, children [WavePlayer_7], properties {pitchMin=0.9}}
 {id=Modulator_4, children [WavePlayer_8], properties {pitchMin=0.9}}
 {id=QualityLevel_0, children [Mixer_7,Mixer_0,Modulator_3,Modulator_4]}
 {id=Random_1, children [WavePlayer_4,WavePlayer_5], properties {weights [1,1]}}
 {id=Random_2, children [WavePlayer_2,WavePlayer_3], properties {bRandomizeWithoutReplacement=false, weights [1,1]}}
 {id=Random_40, children [WavePlayer_138, WavePlayer_137], properties {weights [1,1]}}
 {id=Random_41, children [WavePlayer_141, WavePlayer_142, WavePlayer_143], properties {weights [1,1,1]}}
 {id=Random_42, children [WavePlayer_139, WavePlayer_140], properties {weights [1,1]}}
 {id=WavePlayer_1, properties {soundWaveAssetPtr=/Game/Sounds/Fort_Impact_Sounds/Demolition/Defaults/wood_impact_stereo_01}}
 {id=WavePlayer_137, properties {soundWaveAssetPtr=/Game/Sounds/Fort_Impact_Sounds/Demolition/Defaults/wood_impact_stereo_04}}
 {id=WavePlayer_138, properties {soundWaveAssetPtr=&wood_impact_stereo_01}}
 {id=WavePlayer_139, properties {soundWaveAssetPtr=&/Wood_Small_Shatter01}}
 {id=WavePlayer_140, properties {soundWaveAssetPtr=&/Wood_Small_Shatter02}}
 {id=WavePlayer_141, properties {soundWaveAssetPtr=&/Wood_Small_Shatter03}}
 {id=WavePlayer_142, properties {soundWaveAssetPtr=&/Wood_Small_Shatter04}}
 {id=WavePlayer_143, properties {soundWaveAssetPtr=&/Wood_Small_Shatter05}}
 {id=WavePlayer_2, properties {soundWaveAssetPtr=&Wood_Small_Shatter01}}
 {id=WavePlayer_3, properties {soundWaveAssetPtr=&Wood_Small_Shatter02}}
 {id=WavePlayer_4, properties {soundWaveAssetPtr=&Wood_Small_Shatter03}}
 {id=WavePlayer_5, properties {soundWaveAssetPtr=&Wood_Small_Shatter04}}
 {id=WavePlayer_7, properties {soundWaveAssetPtr=&/Wood_Small_ImpactFinal_opt}}
 {id=WavePlayer_8, properties {soundWaveAssetPtr=&Wood_Small_ImpactFinal_opt}}
]
properties {
 attenuationSettings=/Game/Sounds/Attenuation/Fort_Demolition_Destroy_Medium_Sized_Object_Att
 maxDistance=4000
 soundClassObject=&Fort_Impacts_PSM
 volumeMultiplier=0.48
 concurrencyOverrides {resolutionRule=PreventNew}
 concurrencySet [&CONC_Object_Impacts_FourMax]
}
